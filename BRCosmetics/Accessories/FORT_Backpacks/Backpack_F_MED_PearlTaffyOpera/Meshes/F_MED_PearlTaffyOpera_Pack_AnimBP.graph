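# Animation Blueprint for the PearlTaffyOpera back bling, offsetting the pack on the player's back.
# The AnimGraph chains four additive Modify Bone nodes that shift and roll the pack bone over the input pose, with interpolated alpha settings.
# The EventGraph measures the owner mesh's spine_05 transform delta against spine_01 and pelvis.
# It combines these alphas with skydive-floating and parachute states into pose flags, and also sets a not-skydiving flag.
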AnimBlueprint /BRCosmetics/Accessories/FORT_Backpacks/Backpack_F_MED_PearlTaffyOpera/Meshes/F_MED_PearlTaffyOpera_Pack_AnimBP
target_skeleton=&/F_MED_PearlTaffyOpera_Pack_Skeleton
graphs [
 AnimGraph {nodes [
   n1 = "Transform (Modify) Bone - Bone: pack"(ComponentPose=Local To Component(LocalPose=Input Pose()), Translation="0,2.5,0", Rotation="0,0,0") {
    rotationMode=BMM_Additive
    translationMode=BMM_Additive
    alphaScaleBiasClamp {bInterpResult=true, interpSpeedDecreasing=5, interpSpeedIncreasing=5}
    boneToModify {boneName=pack}
    translation {y=2.5}
   }
   n2 = "Transform (Modify) Bone - Bone: pack"(ComponentPose=@n1, Translation="0,-2,1.9", Rotation="0,0,6.5") {
    rotationMode=BMM_Additive
    translationMode=BMM_Additive
    alphaScaleBiasClamp {bInterpResult=true, interpSpeedDecreasing=6}
    boneToModify {boneName=pack}
    rotation {roll=6.5}
    translation {y=-2, z=1.9}
   }
   n3 = "Transform (Modify) Bone - Bone: pack"(ComponentPose=@n2, Translation="0,0,0", Rotation="0,0,8") {
    rotationMode=BMM_Additive
    translationMode=BMM_Additive
    alphaScaleBiasClamp {bInterpResult=true, interpSpeedDecreasing=6}
    boneToModify {boneName=pack}
    rotation {roll=8}
   }
   n4 = "Transform (Modify) Bone - Bone: pack"(ComponentPose=@n3, Translation="0,-1.5,2", Rotation="0,0,5") {
    rotationMode=BMM_Additive
    translationMode=BMM_Additive
    alphaScaleBiasClamp {bInterpResult=true, interpSpeedDecreasing=6}
    boneToModify {boneName=pack}
    rotation {roll=5}
    translation {y=-1.5, z=2}
   }
   Output Pose(Result=Component To Local(ComponentPose=@n4)) {layerGroup=None}
  ]}
 EventGraph {nodes [
   Event BlueprintUpdateAnimation()
   > n1 = Sequence()
   n2 = GetOwnerSkeletalMeshComponent()
   > n3 = Set Alpha spine rotations(Alpha spine rotations:float=GetAlphaFromReferenceTransformDelta(SkeletalMeshComponent=@n2, BoneName=spine_05, BaseName=spine_01, DeltaComponentType=EulerX, InputRange=(X=5,Y=13), TwistAxis="1,0,0")) -> (Output_Get:float)
   > Set Bool spine rotations(Bool spine rotations:bool=AND Boolean(A="To Bool (Double)"!(Value:double=@n3) -> (ReturnValue:bool), B=Get bIsSkydiveFloating())) -> (Output_Get:bool)
   > n4 = Set Alpha pelvis rotations(Alpha pelvis rotations:float=GetAlphaFromReferenceTransformDelta(SkeletalMeshComponent=@n2, BoneName=spine_05, BaseName=pelvis, InputRange=(X=-30,Y=-40), TwistAxis="1,0,0")) -> (Output_Get:float)
   > Set Bool pelvis rotations(Bool pelvis rotations:bool=AND Boolean(A="To Bool (Double)"(Value=@n4), B=Get bIsParachuteOpened())) -> (Output_Get:bool)
   Set not skydiving(execute=@n1.then_1, not skydiving:bool=NOT Boolean(A=Get bIsSkyDiving())) -> (Output_Get:bool)
   TryGetPawnOwner()
  ]}
]
variables [
 Alpha pelvis rotations {type=number}
 Alpha spine rotations {type=number}
 Bool pelvis rotations {type=boolean}
 Bool spine rotations {type=boolean}
 not skydiving {type=boolean}
]
